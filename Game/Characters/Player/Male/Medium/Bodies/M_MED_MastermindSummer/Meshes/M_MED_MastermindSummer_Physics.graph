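# Physics asset for the M_MED_MastermindSummer player body, simulating a dangling clip with two hooks and a three-segment rope at the pelvis.
# Capsule bodies cover the accessory chain plus pelvis and left thigh, with zero-size capsules on the MSTR root bones.
# Clip, hook and rope joints have 45-degree swing and 2-degree twist limit values.
# The clip root and thigh attach to the pelvis with free angular motion.
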PhysicsAsset /Game/Characters/Player/Male/Medium/Bodies/M_MED_MastermindSummer/Meshes/M_MED_MastermindSummer_Physics
bodies {
 dyn_Clip_MSTR [{shapeType=Capsule, length=0, radius=0, shapeName=dyn_Clip_MSTR_capsule, rotation=<rot 0 -90 0>}]
 dyn_Clip_jnt [{shapeType=Capsule, length=4, radius=2.65, shapeName=dyn_Clip_jnt_capsule, center=<6 0 0.25>, rotation=<rot 90 0 0>}]
 dyn_Hook01_MSTR [{shapeType=Capsule, length=0, radius=0, shapeName=dyn_Hook01_MSTR_capsule, rotation=<rot 0 -90 0>}]
 dyn_Hook01_jnt [{shapeType=Capsule, length=8, radius=0.75, shapeName=dyn_Hook01_jnt_capsule, center=<6 0 0.2>, rotation=<rot 90 0 0>}]
 dyn_Hook02_MSTR [{shapeType=Capsule, length=0, radius=0, shapeName=dyn_Hook02_MSTR_capsule, rotation=<rot 0 -90 0>}]
 dyn_Hook02_jnt [{shapeType=Capsule, length=5, radius=0.75, shapeName=dyn_Hook02_jnt_capsule, center=<5 0 0.2>, rotation=<rot 90 0 0>}]
 dyn_Rope_01_jnt [{shapeType=Capsule, length=12, radius=1.75, shapeName=dyn_Rope_01_jnt_capsule, center=<4 0 0.4>, rotation=<rot 0 0 90>}]
 dyn_Rope_02_jnt [{shapeType=Capsule, length=12, radius=1.75, shapeName=dyn_Rope_02_jnt_capsule, center=<4 0 0.5>, rotation=<rot 0 0 90>}]
 dyn_Rope_03_jnt [{shapeType=Capsule, length=12, radius=1.75, shapeName=dyn_Rope_03_jnt_capsule, center=<4 0 0.4>, rotation=<rot 0 0 90>}]
 dyn_Rope_MSTR [{shapeType=Capsule, length=0, radius=0, shapeName=dyn_Rope_MSTR_capsule, rotation=<rot 0 -90 0>}]
 pelvis [{shapeType=Capsule, length=7, radius=8, shapeName=pelvis_capsule}]
 thigh_l [{shapeType=Capsule, length=30, radius=6.15, shapeName=thigh_l_capsule, center=<-20 0 1>, rotation=<rot 0 90 90>}]
}
constraints [
 {bone1Name=dyn_Clip_jnt, bone2Name=dyn_Clip_MSTR, swing1LimitDegrees=45, swing2LimitDegrees=45, twistLimitDegrees=2}
 {bone1Name=dyn_Hook01_MSTR, bone2Name=dyn_Clip_jnt, swing1LimitDegrees=45, swing2LimitDegrees=45, twistLimitDegrees=2}
 {bone1Name=dyn_Hook01_jnt, bone2Name=dyn_Hook01_MSTR, swing1LimitDegrees=45, swing2LimitDegrees=45, twistLimitDegrees=2}
 {bone1Name=dyn_Hook02_jnt, bone2Name=dyn_Hook02_MSTR, swing1LimitDegrees=45, swing2LimitDegrees=45, twistLimitDegrees=2}
 {bone1Name=dyn_Hook02_MSTR, bone2Name=dyn_Clip_jnt, swing1LimitDegrees=45, swing2LimitDegrees=45, twistLimitDegrees=2}
 {bone1Name=dyn_Rope_MSTR, bone2Name=dyn_Clip_jnt, swing1LimitDegrees=45, swing2LimitDegrees=45, twistLimitDegrees=2}
 {bone1Name=dyn_Rope_01_jnt, bone2Name=dyn_Rope_MSTR, swing1LimitDegrees=45, swing2LimitDegrees=45, twistLimitDegrees=2}
 {bone1Name=dyn_Rope_02_jnt, bone2Name=dyn_Rope_01_jnt, swing1LimitDegrees=45, swing2LimitDegrees=45, twistLimitDegrees=2}
 {bone1Name=dyn_Rope_03_jnt, bone2Name=dyn_Rope_02_jnt, swing1LimitDegrees=45, swing2LimitDegrees=45, twistLimitDegrees=2}
 {
  bone1Name=thigh_l
  bone2Name=pelvis
  swing1LimitDegrees=45
  swing1Motion=Free
  swing2LimitDegrees=45
  swing2Motion=Free
  twistLimitDegrees=45
  twistMotion=Free
 }
 {
  bone1Name=dyn_Clip_MSTR
  bone2Name=pelvis
  swing1LimitDegrees=45
  swing1Motion=Free
  swing2LimitDegrees=45
  swing2Motion=Free
  twistLimitDegrees=45
  twistMotion=Free
 }
]
