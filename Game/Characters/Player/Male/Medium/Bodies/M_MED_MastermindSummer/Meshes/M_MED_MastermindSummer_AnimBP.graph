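# Anim Blueprint for the MastermindSummer male player body, copying the attached parent mesh's pose and adding secondary motion.
# A deform Control Rig feeds pelvis-space RigidBody physics, a second Control Rig, then spring-driven AnimDynamics for tank, knife, grenades and holster.
# Weighted bone constraints spread tube joints between tank and head; a final additive rotation adjusts the rope in the front end.
# Each update sets the RigidBody gravity override, via a deprecated function, and the front-end alpha.
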AnimBlueprint /Game/Characters/Player/Male/Medium/Bodies/M_MED_MastermindSummer/Meshes/M_MED_MastermindSummer_AnimBP
target_skeleton=&/M_MED_MastermindSummer_Skeleton
graphs [
 AnimGraph {nodes [
   n1 = Control Rig(Source=Copy Pose From Mesh() {bUseAttachedParent=true}) {
    alphaBoolBlend {blendOption=Linear}
    controlRigAssetReference {blueprintRigClass=&/M_MED_Mastermind_SummerDeformRig_ControlRig.M_MED_Mastermind_SummerDeformRig_ControlRig_C}
   }
   n2 = RigidBody!(ComponentPose=Local To Component(LocalPose=@n1), Alpha, OverrideWorldGravity:Vector=Get GravityOverride(), ExternalForce) -> (Pose) {
    bClampLinearTranslationLimitToRefPose=true
    bOverrideWorldGravity=true
    overridePhysicsAsset=&/M_MED_MastermindSummer_Physics
    simulationSpace=BaseBoneSpace
    alphaBoolBlend {blendOption=Linear}
    alphaScaleBiasClamp {interpSpeedIncreasing=2}
    baseBoneRef {boneName=pelvis}
    componentAppliedLinearAccClamp=<3000 3000 3000>
    componentLinearAccScale=<0.2 0.2 0.2>
    componentLinearVelScale=<0.4 0.3 0.3>
    simSpaceSettings {
     maxAngularAcceleration=1000
     maxAngularVelocity=3
     maxLinearAcceleration=700
     maxLinearVelocity=500
     velocityScaleZ=0
     externalAngularVelocity {z=0.1}
     externalLinearDragV=<0.2 0.2>
    }
   }
   n3 = Control Rig(Source=Component To Local(ComponentPose=@n2)) {
    alphaBoolBlend {blendOption=Linear}
    controlRigAssetReference {blueprintRigClass=&/M_MED_MastermindSummer_CtrlRig.M_MED_MastermindSummer_CtrlRig_C}
   }
   n4 = Anim Dynamics - Bone: dyn_Tank_MSTR(ComponentPose=Local To Component(LocalPose=@n3), Alpha=1, ExternalForce="0,0,0") {
    angularDampingOverride=0.9999
    angularSpringConstant=75
    bAngularSpring=true
    bOverrideAngularDamping=true
    bUseGravityOverride=true
    simulationSpace=RootRelative
    alphaBoolBlend {blendOption=Linear}
    boundBone {boneName=dyn_Tank_MSTR}
    componentAppliedLinearAccClamp=<3000 3000 3000>
    componentLinearAccScale=<0.3 0.3 0.3>
    componentLinearVelScale=<0.3 0.3 0.3>
    gravityOverride {z=-980}
    physicsBodyDefinitions [{
      boundBone {boneName=dyn_Tank_MSTR}
      boxExtents=<10 10 10>
      constraintSetup {angularLimitsMax=<2 2 10>, angularLimitsMin=<-2 -2 -10>, angularTarget=<1 0 0>}
      localJointOffset=<0 0 0>
     }]
    relativeSpaceBone {boneName=calf_l}
   }
   n5 = Anim Dynamics - Bone: dyn_knife_MSTR(ComponentPose=@n4, Alpha=1, ExternalForce="0,0,0") {
    angularDampingOverride=0.9999
    angularSpringConstant=2500
    bAngularSpring=true
    bOverrideAngularDamping=true
    bUseGravityOverride=true
    simulationSpace=RootRelative
    alphaBoolBlend {blendOption=Linear}
    boundBone {boneName=dyn_knife_MSTR}
    componentAppliedLinearAccClamp=<3000 3000 3000>
    componentLinearAccScale=<0.3 0.3 0.3>
    componentLinearVelScale=<0.3 0.3 0.3>
    gravityOverride {z=-980}
    physicsBodyDefinitions [{
      boundBone {boneName=dyn_knife_MSTR}
      boxExtents=<10 10 10>
      constraintSetup {angularLimitsMax=<2 5 0>, angularLimitsMin=<-2 -5 0>, angularTarget=<1 0 0>}
      localJointOffset=<0 0 0>
     }]
    relativeSpaceBone {boneName=calf_l}
   }
   n6 = Anim Dynamics - Bone: Grenade_01_jnt(ComponentPose=@n5, Alpha=0.5, ExternalForce="0,0,0") {
    angularDampingOverride=0.9999
    angularSpringConstant=500
    bAngularSpring=true
    bOverrideAngularDamping=true
    bUseGravityOverride=true
    simulationSpace=RootRelative
    alphaBoolBlend {blendOption=Linear}
    boundBone {boneName=Grenade_01_jnt}
    componentAppliedLinearAccClamp=<3000 3000 3000>
    componentLinearAccScale=<0.3 0.3 0.3>
    componentLinearVelScale=<0.3 0.3 0.3>
    gravityOverride {z=-980}
    physicsBodyDefinitions [{
      boundBone {boneName=Grenade_01_jnt}
      boxExtents=<10 10 10>
      constraintSetup {angularLimitsMax=<5 5 10>, angularLimitsMin=<-5 -5 -10>, angularTarget=<1 0 0>}
      localJointOffset=<0 0 0>
     }]
    relativeSpaceBone {boneName=calf_l}
   }
   n7 = Anim Dynamics - Bone: Grenade_02_jnt(ComponentPose=@n6, Alpha=1, ExternalForce="0,0,0") {
    angularDampingOverride=0.9999
    angularSpringConstant=1000
    bAngularSpring=true
    bOverrideAngularDamping=true
    bUseGravityOverride=true
    simulationSpace=RootRelative
    alphaBoolBlend {blendOption=Linear}
    boundBone {boneName=Grenade_02_jnt}
    componentAppliedLinearAccClamp=<3000 3000 3000>
    componentLinearAccScale=<0.3 0.3 0.3>
    componentLinearVelScale=<0.3 0.3 0.3>
    gravityOverride {z=-980}
    physicsBodyDefinitions [{
      boundBone {boneName=Grenade_02_jnt}
      boxExtents=<10 10 10>
      constraintSetup {angularLimitsMax=<5 5 10>, angularLimitsMin=<-5 -5 -10>, angularTarget=<1 0 0>}
      localJointOffset=<0 0 0>
     }]
    relativeSpaceBone {boneName=calf_l}
   }
   n8 = Anim Dynamics - Bone: dyn_Holster_MSTR(ComponentPose=@n7, Alpha=0.25, ExternalForce="0,0,0") {
    angularDampingOverride=0.9999
    angularSpringConstant=2500
    bAngularSpring=true
    bOverrideAngularDamping=true
    bUseGravityOverride=true
    simulationSpace=RootRelative
    alphaBoolBlend {blendOption=Linear}
    boundBone {boneName=dyn_Holster_MSTR}
    componentAppliedLinearAccClamp=<3000 3000 3000>
    componentLinearAccScale=<0.3 0.3 0.3>
    componentLinearVelScale=<0.3 0.3 0.3>
    gravityOverride {z=-980}
    physicsBodyDefinitions [{
      boundBone {boneName=dyn_Holster_MSTR}
      boxExtents=<10 10 10>
      constraintSetup {angularLimitsMax=<10 20 10>, angularLimitsMin=<-10 -20 -10>, angularTarget=<1 0 0>}
      localJointOffset=<0 0 0>
     }]
    relativeSpaceBone {boneName=calf_l}
   }
   n9 = "Constraint - Tube_01_jnt (P)"(ComponentPose=@n8, Alpha=1, ConstraintWeights_0=1) {
    alphaBoolBlend {blendOption=Linear}
    boneToModify {boneName=Tube_01_jnt}
    constraintSetup [{targetBone {boneName=dyn_Tank_jnt}}]
    constraintWeights [1]
   }
   n10 = "Constraint - Tube_07_jnt (P)"(ComponentPose=@n9, Alpha=1, ConstraintWeights_0=1) {
    alphaBoolBlend {blendOption=Linear}
    boneToModify {boneName=Tube_07_jnt}
    constraintSetup [{targetBone {boneName=head}}]
    constraintWeights [1]
   }
   n11 = "Constraint - Tube_04_jnt (PP)"!(ComponentPose=@n10, Alpha=1, ConstraintWeights_0=0.5, ConstraintWeights_1:float=0.5) -> (Pose) {
    alphaBoolBlend {blendOption=Linear}
    boneToModify {boneName=Tube_04_jnt}
    constraintSetup [{targetBone {boneName=Tube_01_jnt}}, {targetBone {boneName=Tube_07_jnt}}]
    constraintWeights [1,1]
   }
   n12 = "Constraint - Tube_02_jnt (PP)"!(ComponentPose=@n11, Alpha=1, ConstraintWeights_0=0.7, ConstraintWeights_1:float=0.3) -> (Pose) {
    alphaBoolBlend {blendOption=Linear}
    boneToModify {boneName=Tube_02_jnt}
    constraintSetup [{targetBone {boneName=Tube_01_jnt}}, {targetBone {boneName=Tube_04_jnt}}]
    constraintWeights [1,1]
   }
   n13 = "Constraint - Tube_03_jnt (PP)"!(ComponentPose=@n12, Alpha=1, ConstraintWeights_0=0.3, ConstraintWeights_1:float=0.7) -> (Pose) {
    alphaBoolBlend {blendOption=Linear}
    boneToModify {boneName=Tube_03_jnt}
    constraintSetup [{targetBone {boneName=Tube_01_jnt}}, {targetBone {boneName=Tube_04_jnt}}]
    constraintWeights [1,1]
   }
   n14 = "Constraint - Tube_05_jnt (PP)"!(ComponentPose=@n13, Alpha=1, ConstraintWeights_0=0.7, ConstraintWeights_1:float=0.3) -> (Pose) {
    alphaBoolBlend {blendOption=Linear}
    boneToModify {boneName=Tube_05_jnt}
    constraintSetup [{targetBone {boneName=Tube_04_jnt}}, {targetBone {boneName=Tube_07_jnt}}]
    constraintWeights [1,1]
   }
   n15 = "Constraint - Tube_06_jnt (PP)"!(ComponentPose=@n14, Alpha=1, ConstraintWeights_0=0.3, ConstraintWeights_1:float=0.7) -> (Pose) {
    alphaBoolBlend {blendOption=Linear}
    boneToModify {boneName=Tube_06_jnt}
    constraintSetup [{targetBone {boneName=Tube_04_jnt}}, {targetBone {boneName=Tube_07_jnt}}]
    constraintWeights [1,1]
   }
   n16 = "Transform (Modify) Bone - Bone: dyn_Rope_MSTR"(
    ComponentPose=@n15
    Alpha=Get Is In Front End Alpha() -> (Is In Front End Alpha:double)
    Translation="0,0,0"
    Rotation="-2,0,0"
    Scale="1,1,1"
   ) {
    rotationMode=BMM_Additive
    rotationSpace=BCS_BoneSpace
    translationMode=BMM_Additive
    translationSpace=BCS_BoneSpace
    alphaBoolBlend {blendOption=Linear}
    boneToModify {boneName=dyn_Rope_MSTR}
    rotation {pitch=-2}
   }
   Output Pose(Result=Component To Local(ComponentPose=@n16)) {layerGroup=None}
  ]}
 EventGraph {nodes [
   Event BlueprintUpdateAnimation()
   > n1 = Sequence!(execute) -> (then_0, then_1, then_2:Exec)
   > n2 = Gravity Override State Function DEPRECATED(CharacterPartAnimInstance=Self-Reference(), GameTimeInSeconds=GetGameTimeInSeconds(), GravityOverrideParams=Get Gravity Override Params())
   > Set GravityOverride(GravityOverride=@n2)
   Set Is In Front End Alpha(execute=@n1.then_1, Is In Front End Alpha:double="To Float (Boolean)"(InBool=Get bIsInFrontEnd())) -> (Output_Get:double)
  ]}
]
variables [
 Gravity Override Params {type=GravityOverrideParamsStruct}
 GravityOverride {type=Vector}
 Is In Front End Alpha {type=number}
]
